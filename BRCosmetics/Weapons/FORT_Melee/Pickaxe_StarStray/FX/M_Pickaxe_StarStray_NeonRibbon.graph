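# Unlit translucent Niagara ribbon material for the neon ribbon effect on the StarStray pickaxe.
# Particle color RGB runs through the shared emissive adjustment function to drive the glow.
# The texture-free ribbon profile adds a dim wide falloff to a narrow bright core, both sphere masks centered on the V coordinate.
# That profile is multiplied by particle alpha and a parameterized depth fade, which the mobile shading path replaces with a constant.
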Material /BRCosmetics/Weapons/FORT_Melee/Pickaxe_StarStray/FX/M_Pickaxe_StarStray_NeonRibbon
graph {
 properties {bUsedWithNiagaraRibbons=true, blendMode=BLEND_Translucent, pixelDepthOffsetMode=PDOM_Legacy, shadingModel=MSM_Unlit, usedShadingModels=Unlit}
 nodes [
  n1 = ParticleColor()
  MaterialFunctionCall(Emissive=@n1.RGB) {materialFunction=&MF_FX_EmissiveAdjustments}
  n2 = ComponentMask(None=TextureCoordinate()) {g=true}
  n3 = Constant() {r=0.5}
  n4 = Multiply(A=SphereMask(A=@n2, B=@n3, Radius=@n3) {hardnessPercent=0}) {constB=0.25}
  n5 = SphereMask(A=@n2, B=Constant() {r=0.5}, Radius=Constant() {r=0.1}) {hardnessPercent=50}
  n6 = DepthFade(FadeDistance=ScalarParameter() {defaultValue=8, parameterName=FadeDist})
  n7 = Constant() {r=1}
  Multiply(A=@n1.A, B=Multiply(A=Add(A=@n4, B=@n5), B=ShadingPathSwitch(Default=@n6, Deferred=@n6, Mobile=@n7) {desc=Migrated from FeatureLevelSwitch}))
 ]
}
