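# Translucent Niagara sprite material for the Star Stray pickaxe's radial volume bar effect, reusing Cactus Dancer impact textures.
# Particle color tints a texture, whose saturation a negative desaturation parameter boosts by default, before it enters the shared emissive adjustment function.
# A second texture's smoothstepped alpha is eroded by subtracting a dynamic parameter, then multiplied by particle alpha.
# That alpha is softened by depth fade, which the mobile shading path skips.
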Material /BRCosmetics/Weapons/FORT_Melee/Pickaxe_StarStray/FX/M_StarStray_RadialVolumeBar
graph {
 properties {bUsedWithNiagaraSprites=true, blendMode=BLEND_Translucent, pixelDepthOffsetMode=PDOM_Legacy, usedShadingModels=Default Lit}
 nodes [
  n1 = ParticleColor()
  n2 = TextureSample() {samplerType=SAMPLERTYPE_LinearColor, texture=/BRCosmetics/Weapons/FORT_Melee/Pickaxe_Cactus_Dancer_Female/FX/T_FX_CactusDancer_Impact_03}
  n3 = ScalarParameter() {defaultValue=-0.2, parameterName=Param_2}
  MaterialFunctionCall(Emissive=Multiply(A=@n1.RGB, B=Desaturation(None=@n2.RGB, Fraction=@n3))) {materialFunction=&MF_FX_EmissiveAdjustments}
  n4 = TextureSample() {samplerType=SAMPLERTYPE_LinearColor, texture=/BRCosmetics/Weapons/FORT_Melee/Pickaxe_Cactus_Dancer_Female/FX/T_FX_CactusDancer_Impact}
  n5 = Subtract(
   A=SmoothStep(Value=@n4.A) {constMax=0.5}
   B=DynamicParameter() {defaultValue=<rgba 0 1 1 1>, paramNames [Erosion,Param2,Param3,Param4]}.Erosion
  )
  n6 = DepthFade(FadeDistance=ScalarParameter() {defaultValue=40, parameterName=FadeDistance})
  Multiply(
   A=Multiply(A=@n1.A, B=Saturate(None=@n5))
   B=ShadingPathSwitch(Default=@n6, Deferred=@n6, Mobile=Constant() {r=1}) {desc=Migrated from FeatureLevelSwitch}
  )
  OneMinus(None=@n4.R)
 ]
}
